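# Nanite-enabled Asteria prototype surface material that tints world-aligned checker and grid textures with a color parameter.
# Two world-aligned projections, the checker at double the grid's tile size, fade by intensity parameters and multiply into base color.
# Scalars and static switches set roughness, specular, metalness and an optional normal map.
# An optional up-vector blend recolors upward faces (snow, grass, dirt) with a tinted grid that omits the checker.
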Material /Game/Environments/Asteria/Materials/Proto/M_Asteria_Proto_B
graph {
 properties {bUseMaterialAttributes=true, bUsedWithNanite=true, pixelDepthOffsetMode=PDOM_Legacy, usedShadingModels=Default Lit}
 nodes [
  n1 = TextureObject() {texture=/Game/Environments/Asteria/Textures/T_Asteria_Checker_A_D}
  n2 = Constant3Vector() {constant=<rgba 512 256 384 0>}
  n3 = StaticBool() {value=true}
  n4 = Constant3Vector() {constant=<rgba 512 512 512 0>}
  n5 = MaterialFunctionCall!(
   Diffuse Texture Object=@n1
   Spec Texture Object
   TextureSize=Multiply(A=@n2) {constB=2}
   --------------------------
   Use seperate settings for Z projection=@n3
   Use seperate texture for Z projection=@n3
   Texture Size Z projection=Multiply(A=@n4) {constB=2}
   Z Plane Diffuse Texture Object=@n1
   Z Plane Spec Texture Object
   "--------------------------      "
   ProjectionTransitionContrast
   Use Texture Alpha Heightmap to Refine the Transistion
   HeightLerpTransitionContrast
   PickTheDiffuseHeightChannel
   TheDiffuseHeightChannel
   "--------------------------       "
   NormalTextureObject
   World Space Normals
   "--------------------------  "
   Export Float 4 Diffuse
   Export Float 4 Spec
   "--------------------------     "
   WorldPosition
   World Space Normal
  ) {materialFunction=/Engine/Functions/Engine_MaterialFunctions02/Texturing/WorldAlignedTextures_Complex}
  n6 = OneMinus(None=ScalarParameter() {defaultValue=0.5, parameterName=Checker Intensity})
  n7 = VectorParameter() {parameterName=Color, defaultValue=<rgba 0.1302 0.1302 0.1302 1>}
  n8 = StaticBool() {value=true}
  n9 = MaterialFunctionCall(
   Diffuse Texture Object=TextureObject() {texture=/Game/Environments/Asteria/Textures/T_Asteria_Grid_A_D}
   TextureSize=@n2
   Use seperate settings for Z projection=@n8
   Use seperate texture for Z projection=@n8
   Texture Size Z projection=@n4
   Z Plane Diffuse Texture Object=TextureObject() {texture=/Game/Environments/Asteria/Textures/T_Asteria_Grid_B_D}
  ) {materialFunction=&WorldAlignedTextures_Complex}
  n10 = OneMinus(None=ScalarParameter() {defaultValue=1, parameterName=Checker Intensity B})
  n11 = LinearInterpolate(A=@n9.XYZ Diffuse Texture, Alpha=@n10)
  n12 = Constant() {r=1}
  n13 = TextureSampleParameter2D() {parameterName=Normal, samplerType=SAMPLERTYPE_Normal, texture=/Game/Environments/Apollo/Rocks/Natural_Rocks/Textures/T_NaturalRock_AB_N}
  n14 = StaticSwitchParameter(True=@n13.RGB, False=TextureSample() {samplerType=SAMPLERTYPE_Normal, texture=&FlatNormalBlue}.RGB) {parameterName=Use Normal Map?}
  n15 = MakeMaterialAttributes!(
   BaseColor=Multiply(A=LinearInterpolate(A=@n5.XYZ Diffuse Texture, Alpha=@n6), B=Multiply(A=@n7.RGB, B=@n11))
   Metallic=StaticSwitchParameter(True=@n12, False=Constant()) {parameterName=Metal?}
   Specular=StaticSwitchParameter(True=@n12, False=ScalarParameter() {defaultValue=0.5, parameterName=Specular Value}) {parameterName=Metal?}
   Roughness=ScalarParameter() {defaultValue=0.5, parameterName=Roughness Value}
   Anisotropy
   EmissiveColor
   Opacity
   OpacityMask
   Normal=@n14
   Tangent
   WorldPositionOffset
   SubsurfaceColor
   ClearCoat
   ClearCoatRoughness
   AmbientOcclusion
   Refraction
   CustomizedUVs_0
   CustomizedUVs_1
   CustomizedUVs_2
   CustomizedUVs_3
   CustomizedUVs_4
   CustomizedUVs_5
   CustomizedUVs_6
   CustomizedUVs_7
   PixelDepthOffset
   ShadingModel
   Displacement
  )
  n16 = Multiply(
   A=LinearInterpolate(A=@n9.XYZ Diffuse Texture, Alpha=@n10)
   B=VectorParameter() {group=Up Vector, parameterName=Up Vector Color, defaultValue=<rgba 1 0 0 1>}.RGB
  )
  n17 = ScalarParameter() {defaultValue=3, group=Up Vector, parameterName=Blend Sharpness}
  n18 = ScalarParameter() {group=Up Vector, parameterName=Up Vector Blend Bias}
  n19 = StaticBool() {value=true}
  StaticSwitchParameter(
   True=MaterialFunctionCall!(
    Base Material=@n15
    Top Material=SetMaterialAttributes!(MaterialAttributes=@n15, Base Color=@n16) {attributeSetTypes [#1]}
    Alpha=MaterialFunctionCall!(In Explicit Normal=@n14, In World Vector, Blend Sharpness=@n17, Blend Bias=@n18, Clamped?=@n19, Alpha) {materialFunction=/Engine/Functions/Engine_MaterialFunctions01/AlphaBlend/WorldAlignedBlend}.w/Vertex Normals
   ) {materialFunction=/Engine/Functions/MaterialLayerFunctions/MatLayerBlend_Standard}.Blended Material
   False=@n15
  ) {group=Up Vector, parameterName="Use Up Vector Blend (snow, grass, dirt etc)"}
 ]
}
