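# Prototype ground material for the Asteria environment, drawing a tinted checker and grid pattern in world space, with Nanite usage enabled.
# Two world-aligned projections, with separate Z-axis settings and adjustable intensities, multiply into base color and, inverted plus an offset, roughness.
# Static switches toggle metallic and a rock normal map, transformed out of tangent space.
# Attributes feed a runtime virtual texture output, while FN_RVT_Switch_Material selects between direct shading and sampling the Apollo landscape RVT.
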Material /Game/Environments/Asteria/Materials/Proto/M_Asteria_Proto_Ground
graph {
 properties {
  bTangentSpaceNormal=false
  bUseMaterialAttributes=true
  bUsedWithNanite=true
  pixelDepthOffsetMode=PDOM_Legacy
  usedShadingModels=Default Lit
 }
 nodes [
  Multiply()
  n1 = TextureObject() {texture=/Game/Environments/Asteria/Textures/T_Asteria_Checker_A_D}
  n2 = Constant3Vector() {constant=<rgba 512 256 384 0>}
  n3 = Multiply(A=@n2) {constB=2}
  n4 = StaticBool() {value=true}
  n5 = Constant3Vector() {constant=<rgba 512 512 512 0>}
  n6 = Multiply(A=@n5) {constB=2}
  n7 = ScalarParameter() {defaultValue=0.5, parameterName=Checker Intensity}
  n8 = TextureObject() {texture=/Game/Environments/Asteria/Textures/T_Asteria_Grid_A_D}
  n9 = StaticBool() {value=true}
  n10 = TextureObject() {texture=/Game/Environments/Asteria/Textures/T_Asteria_Grid_B_D}
  n11 = ScalarParameter() {defaultValue=1, parameterName=Checker Intensity B}
  n12 = Multiply(
   A=LinearInterpolate(
    A=MaterialFunctionCall!(
     Diffuse Texture Object=@n1
     Spec Texture Object
     TextureSize=@n3
     --------------------------
     Use seperate settings for Z projection=@n4
     Use seperate texture for Z projection=@n4
     Texture Size Z projection=@n6
     Z Plane Diffuse Texture Object=@n1
     Z Plane Spec Texture Object
     "--------------------------      "
     ProjectionTransitionContrast
     Use Texture Alpha Heightmap to Refine the Transistion
     HeightLerpTransitionContrast
     PickTheDiffuseHeightChannel
     TheDiffuseHeightChannel
     "--------------------------       "
     NormalTextureObject
     World Space Normals
     "--------------------------  "
     Export Float 4 Diffuse
     Export Float 4 Spec
     "--------------------------     "
     WorldPosition
     World Space Normal
    ) {materialFunction=/Engine/Functions/Engine_MaterialFunctions02/Texturing/WorldAlignedTextures_Complex}.Z Diffuse Texture
    Alpha=OneMinus(None=@n7)
   )
   B=LinearInterpolate(
    A=MaterialFunctionCall(
     Diffuse Texture Object=@n8
     TextureSize=@n2
     Use seperate settings for Z projection=@n9
     Use seperate texture for Z projection=@n9
     Texture Size Z projection=@n5
     Z Plane Diffuse Texture Object=@n10
    ) {materialFunction=&WorldAlignedTextures_Complex}.Z Diffuse Texture
    Alpha=OneMinus(None=@n11)
   )
  )
  n13 = Constant() {r=1}
  n14 = ScalarParameter() {defaultValue=0.25, parameterName=Specular Value}
  n15 = TextureSampleParameter2D() {parameterName=Normal, samplerType=SAMPLERTYPE_Normal, texture=/Game/Environments/Apollo/Rocks/Natural_Rocks/Textures/T_NaturalRock_AB_N}
  n16 = TextureSample() {samplerType=SAMPLERTYPE_Normal, texture=&FlatNormalBlue}
  n17 = MakeMaterialAttributes!(
   BaseColor=Multiply(A=@n12, B=VectorParameter() {parameterName=Color, defaultValue=<rgba 0.1615 0.1112 0.0518 1>}.RGB)
   Metallic=StaticSwitchParameter(True=@n13, False=Constant()) {parameterName=Metal?}
   Specular=@n14
   Roughness=Add(A=OneMinus(None=@n12), B=ScalarParameter() {defaultValue=0.35, parameterName=Roughness Value})
   Anisotropy
   EmissiveColor
   Opacity
   OpacityMask
   Normal=Transform(None=StaticSwitchParameter(True=@n15.RGB, False=@n16.RGB) {parameterName=Use Normal Map?})
   Tangent
   WorldPositionOffset
   SubsurfaceColor
   ClearCoat
   ClearCoatRoughness
   AmbientOcclusion
   Refraction
   CustomizedUVs_0
   CustomizedUVs_1
   CustomizedUVs_2
   CustomizedUVs_3
   CustomizedUVs_4
   CustomizedUVs_5
   CustomizedUVs_6
   CustomizedUVs_7
   PixelDepthOffset
   ShadingModel
   Displacement
  )
  n18 = GetMaterialAttributes(None=@n17) {attributeGetTypes [
    #1
    #2
    #3
    #4
    #5
   ]}
  RuntimeVirtualTextureOutput!(
   BaseColor=@n18.Base Color
   Specular=@n18.Base Color
   Roughness=@n18.Base Color
   Normal=@n18.Base Color
   WorldHeight
   Opacity=@n18.Base Color
   Mask
   Displacement
   Mask4
  )
  n19 = RuntimeVirtualTextureSampleParameter() {materialType=BaseColor_Normal_Specular_YCoCg, parameterName=RVT, virtualTexture=/Game/Athena/Apollo/Maps/Landscape/VT/RVT_Apollo_Landscape}
  n20 = SetMaterialAttributes!(MaterialAttributes, Base Color=@n19.BaseColor, Specular=@n19.BaseColor, Roughness=@n19.BaseColor, Normal=@n19.BaseColor) {attributeSetTypes [
    #1
    #2
    #3
    #4
   ]}
  n21 = MaterialFunctionCall(NoRVT=@n17, RVT=MaterialProxyReplace(Realtime=@n20, MaterialProxy=@n18.MaterialAttributes)) {materialFunction=&FN_RVT_Switch_Material}
  ShadingPathSwitch(Default=@n21.Result, Deferred=@n21.Result, Forward=@n21.Result, Mobile=@n21.Result)
  StaticSwitchParameter(True=@n13, False=@n14) {parameterName=Metal?}
 ]
}
